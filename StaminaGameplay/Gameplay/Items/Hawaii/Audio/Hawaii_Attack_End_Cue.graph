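# Sound cue for the end of the Hawaii item's attack in StaminaGameplay.
# A Random node picks one of three Hawaii_Attack_End waves with equal weight.
# It plays quietly under the gadget-items sound class with skill-use attenuation.
# A concurrency override allows only one instance at a time and blocks new plays while it sounds.
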SoundCue /StaminaGameplay/Gameplay/Items/Hawaii/Audio/Hawaii_Attack_End_Cue
root=Random_1
nodes [
 {id=Random_1, children [WavePlayer_0,WavePlayer_1,WavePlayer_2], properties {weights [1,1,1]}}
 {id=WavePlayer_0, properties {soundWaveAssetPtr=&/Hawaii_Attack_End_01}}
 {id=WavePlayer_1, properties {soundWaveAssetPtr=&/Hawaii_Attack_End_02}}
 {id=WavePlayer_2, properties {soundWaveAssetPtr=&/Hawaii_Attack_End_03}}
]
properties {
 attenuationSettings=/Game/Sounds/Attenuation/Fort_SkillUse_Att
 bOverrideConcurrency=true
 maxDistance=4000
 soundClassObject=&Fort_Gadgets_Items
 volumeMultiplier=0.24
 concurrencyOverrides {maxCount=1, resolutionRule=PreventNew}
}
