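# Sound cue for the collision sound of a Hawaii item beam battle.
# A random node picks one of three equally weighted collision waves, fed through a modulator with pitch fixed at 1.
# It plays quietly in the explosives sound class using the Hawaii impact attenuation asset.
# Overridden concurrency allows one instance per owner, stopping the farthest and then preventing new ones.
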SoundCue /StaminaGameplay/Gameplay/Items/Hawaii/Audio/Hawaii_BeamBattle_Collision_Cue
root=Modulator_0
nodes [
 {id=Modulator_0, children [Random_1], properties {pitchMax=1, pitchMin=1}}
 {id=Random_1, children [WavePlayer_3,WavePlayer_4,WavePlayer_5], properties {weights [1,1,1]}}
 {id=WavePlayer_3, properties {soundWaveAssetPtr=&/Hawaii_BeamBattle_Collision_01}}
 {id=WavePlayer_4, properties {soundWaveAssetPtr=&/Hawaii_BeamBattle_Collision_02}}
 {id=WavePlayer_5, properties {soundWaveAssetPtr=&/Hawaii_BeamBattle_Collision_03}}
]
properties {
 attenuationSettings=/StaminaGameplay/Gameplay/Items/Hawaii/Audio/Settings/Fort_Hawaii_Impact_Att
 bOverrideConcurrency=true
 maxDistance=13000
 soundClassObject=/Game/Sounds/FortSoundClassesAndModes/Fort_Explosives
 volumeMultiplier=0.3
 concurrencyOverrides {bLimitToOwner=true, maxCount=1, resolutionRule=StopFarthestThenPreventNew}
}
