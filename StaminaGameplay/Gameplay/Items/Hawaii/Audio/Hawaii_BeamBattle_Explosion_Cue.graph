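# Sound cue for the explosion in the Hawaii item's beam battle.
# A Random node picks one of three equally weighted explosion waves and feeds it through a Modulator to the output.
# It uses the Hawaii impact attenuation and the explosives sound class at reduced volume.
# A concurrency override allows one instance per owner, stopping the farthest and otherwise preventing new ones.
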SoundCue /StaminaGameplay/Gameplay/Items/Hawaii/Audio/Hawaii_BeamBattle_Explosion_Cue
root=Modulator_0
nodes [
 {id=Modulator_0, children [Random_1]}
 {id=Random_1, children [WavePlayer_3,WavePlayer_4,WavePlayer_5], properties {weights [1,1,1]}}
 {id=WavePlayer_3, properties {soundWaveAssetPtr=&/Hawaii_BeamBattle_Explosion_03}}
 {id=WavePlayer_4, properties {soundWaveAssetPtr=&/Hawaii_BeamBattle_Explosion_01}}
 {id=WavePlayer_5, properties {soundWaveAssetPtr=&/Hawaii_BeamBattle_Explosion_02}}
]
properties {
 attenuationSettings=/StaminaGameplay/Gameplay/Items/Hawaii/Audio/Settings/Fort_Hawaii_Impact_Att
 bOverrideConcurrency=true
 maxDistance=13000
 soundClassObject=/Game/Sounds/FortSoundClassesAndModes/Fort_Explosives
 volumeMultiplier=0.35
 concurrencyOverrides {bLimitToOwner=true, maxCount=1, resolutionRule=StopFarthestThenPreventNew}
}
