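# Sound cue for the Hawaii item's beam, playing one of three equally weighted beam waves.
# An IsLocal branch plays it without an attenuation node for the local player; otherwise an IsEnemy branch routes it through shared beam attenuation.
# Enemy and unset cases keep full volume, while non-enemies are heavily reduced.
# Concurrency is limited per owner, stopping the farthest instance first.
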SoundCue /StaminaGameplay/Gameplay/Items/Hawaii/Audio/Hawaii_Beam_Cue
root=Branch_2
nodes [
 {
  id=Attenuation_0
  children [Random_0]
  properties {attenuationSettings=/StaminaGameplay/Gameplay/Items/Hawaii/Audio/Settings/Fort_Hawaii_Beam_Att}
 }
 {
  id=Attenuation_2
  children [Random_0]
  properties {attenuationSettings=&Fort_Hawaii_Beam_Att}
 }
 {
  id=Attenuation_3
  children [Random_0]
  properties {attenuationSettings=&Fort_Hawaii_Beam_Att}
 }
 {id=Branch_1, children [Modulator_3, Modulator_2, Attenuation_0], properties {boolParameterName=IsEnemy}}
 {id=Branch_2, children [Random_0,Branch_1,Branch_1], properties {boolParameterName=IsLocal}}
 {id=Modulator_2, children [Attenuation_3], properties {pitchMax=1, pitchMin=1, volumeMax=0.3, volumeMin=0.3}}
 {id=Modulator_3, children [Attenuation_2], properties {pitchMax=1, pitchMin=1, volumeMax=1, volumeMin=1}}
 {id=Random_0, children [WavePlayer_0,WavePlayer_1,WavePlayer_2], properties {weights [1,1,1]}}
 {id=WavePlayer_0, properties {soundWaveAssetPtr=&/Hawaii_Beam_03}}
 {id=WavePlayer_1, properties {soundWaveAssetPtr=&/Hawaii_Beam_01}}
 {id=WavePlayer_2, properties {soundWaveAssetPtr=&/Hawaii_Beam_02}}
]
properties {
 bOverrideConcurrency=true
 maxDistance=9700
 soundClassObject=&Fort_Gadgets_Items
 volumeMultiplier=0.85
 concurrencyOverrides {bLimitToOwner=true, maxCount=2, resolutionRule=StopFarthestThenPreventNew}
}
