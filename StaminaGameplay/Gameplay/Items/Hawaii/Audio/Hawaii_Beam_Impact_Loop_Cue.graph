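# Sound cue that plays the looping impact sound for the Hawaii item's beam.
# A looping wave player feeds an enveloper that ramps volume up over a short fade-in at the start.
# The cue uses the Hawaii impact attenuation, a third-person ranged-weapon fire sound class, and an owner-limited concurrency override that stops the farthest instance first.
# Debug is enabled on the cue.
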SoundCue /StaminaGameplay/Gameplay/Items/Hawaii/Audio/Hawaii_Beam_Impact_Loop_Cue
root=Enveloper_0
nodes [
 {
  id=Enveloper_0
  children [WavePlayer_0]
  properties {pitchCurve {editorCurveData {keys [{time=0}]}}, volumeCurve {editorCurveData {keys [<at 0 0>,<at 0.2 1>]}}}
 }
 {id=WavePlayer_0, properties {bLooping=true, soundWaveAssetPtr=&/Hawaii_Beam_Impact_Loop}}
]
properties {
 attenuationSettings=/StaminaGameplay/Gameplay/Items/Hawaii/Audio/Settings/Fort_Hawaii_Impact_Att
 bDebug=true
 bOverrideConcurrency=true
 maxDistance=13000
 soundClassObject=&Fort_WeaponRanged_Fire_3P_PSM
 volumeMultiplier=0.5
 attenuationOverrides {falloffDistance=1500, stereoSpread=512}
 concurrencyOverrides {bLimitToOwner=true, maxCount=2, resolutionRule=StopFarthestThenPreventNew}
}
