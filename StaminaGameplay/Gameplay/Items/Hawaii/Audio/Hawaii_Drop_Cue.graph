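# Sound cue for the drop sound of the Hawaii item in Stamina gameplay.
# A single wave player plays the Hawaii_Drop sound wave directly as the root, with no randomization or modulation.
# It uses the shared Fortnite weapon-drop attenuation and the loot-collect sound class, with reduced volume and a limited audible range.
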SoundCue /StaminaGameplay/Gameplay/Items/Hawaii/Audio/Hawaii_Drop_Cue
root=WavePlayer_0
nodes [{id=WavePlayer_0, properties {soundWaveAssetPtr=&/Hawaii_Drop}}]
properties {
 attenuationSettings=&Fort_Weapon_Drop_Att
 maxDistance=3000
 soundClassObject=&Fort_Loot_Collect
 volumeMultiplier=0.7
}
