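# Sound cue for equipping the Hawaii item, with separate normal and cooldown variants routed through player equip attenuation.
# An EquipAudioParam switch selects the normal mix in its first two slots and the cooldown mix in its last, leaving two slots empty.
# The normal mix layers a random modulated one-shot with a delayed, enveloped equipped loop.
# The cooldown mix pairs a random cooldown one-shot with an enveloped cooldown loop, both loops shaped by pitch and volume curves.
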SoundCue /StaminaGameplay/Gameplay/Items/Hawaii/Audio/Hawaii_Equip_Cue
root=Switch_1
nodes [
 {id=Delay_0, children [Enveloper_0], properties {delayMax=0.2, delayMin=0.2}}
 {
  id=Enveloper_0
  children [WavePlayer_3]
  properties {
   pitchCurve {editorCurveData {keys [<at -0.0001 0.6>, <at 0.5058 0.7553>, {time=1}]}}
   volumeCurve {editorCurveData {keys [<at 0 0.01>,<at 0.6906 0.5495>,<at 1 1>,<at 4 1>,<at 8 0.6>,<at 15 0.3>]}}
  }
 }
 {
  id=Enveloper_1
  children [WavePlayer_4]
  properties {
   volumeMax=0.5
   volumeMin=0.5
   pitchCurve {editorCurveData {keys [<at -0.0001 0.7>]}}
   volumeCurve {editorCurveData {keys [<at 0 0.01>,<at 4.4877 0.5765>,<at 10 0.7>]}}
  }
 }
 {id=Mixer_0, children [Modulator_0,Delay_0], properties {inputVolume [1,0.3]}}
 {id=Mixer_1, children [Enveloper_1,Random_1], properties {inputVolume [1,0.7]}}
 {id=Modulator_0, children [Random_0]}
 {id=Random_0, children [WavePlayer_0,WavePlayer_2,WavePlayer_1], properties {bShouldExcludeFromBranchCulling=true, weights [1,1,1]}}
 {id=Random_1, children [WavePlayer_5,WavePlayer_6,WavePlayer_7], properties {weights [1,1,1]}}
 {id=Switch_1, children [Mixer_0,Mixer_0,null,null,Mixer_1], properties {intParameterName=EquipAudioParam}}
 {id=WavePlayer_0, properties {soundWaveAssetPtr=&/Hawaii_Equip_01}}
 {id=WavePlayer_1, properties {soundWaveAssetPtr=&/Hawaii_Equip_03}}
 {id=WavePlayer_2, properties {soundWaveAssetPtr=&/Hawaii_Equip_02}}
 {id=WavePlayer_3, properties {bLooping=true, soundWaveAssetPtr=&/Hawaii_Equipped_Loop}}
 {id=WavePlayer_4, properties {bLooping=true, soundWaveAssetPtr=&/Hawaii_Equipped_Cooldown_Loop}}
 {id=WavePlayer_5, properties {soundWaveAssetPtr=&/Hawaii_Equip_Cooldown_01}}
 {id=WavePlayer_6, properties {soundWaveAssetPtr=&/Hawaii_Equip_Cooldown_02}}
 {id=WavePlayer_7, properties {soundWaveAssetPtr=&/Hawaii_Equip_Cooldown_03}}
]
properties {
 attenuationSettings=/Game/Sounds/Attenuation/Fort_Player_Equip_Generic_Att
 maxDistance=1800
 soundClassObject=&Fort_Gadgets_Items
 volumeMultiplier=0.4
}
