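# Sound cue for the impact sound of the Hawaii item in StaminaGameplay, classed under Fort_Explosives.
# A Random node picks one of three equally weighted impact waves and feeds a Modulator at the root for pitch and volume variation.
# It uses a dedicated impact attenuation asset, and a concurrency override caps it at two instances per owner, preventing new plays beyond that.
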SoundCue /StaminaGameplay/Gameplay/Items/Hawaii/Audio/Hawaii_Impact_Cue
root=Modulator_0
nodes [
 {id=Modulator_0, children [Random_1]}
 {id=Random_1, children [WavePlayer_0,WavePlayer_1,WavePlayer_2], properties {weights [1,1,1]}}
 {id=WavePlayer_0, properties {soundWaveAssetPtr=&/Hawaii_Impact_01}}
 {id=WavePlayer_1, properties {soundWaveAssetPtr=&/Hawaii_Impact_02}}
 {id=WavePlayer_2, properties {soundWaveAssetPtr=&/Hawaii_Impact_03}}
]
properties {
 attenuationSettings=/StaminaGameplay/Gameplay/Items/Hawaii/Audio/Settings/Fort_Hawaii_Impact_Att
 bOverrideConcurrency=true
 maxDistance=13000
 soundClassObject=/Game/Sounds/FortSoundClassesAndModes/Fort_Explosives
 concurrencyOverrides {bLimitToOwner=true, maxCount=2, resolutionRule=PreventNew}
}
