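# Sound cue for the pickup sound of the Hawaii item in StaminaGameplay.
# A single wave player feeds the Hawaii_PickUp wave straight to the output, with no randomization or modulation.
# It uses the shared weapon pickup attenuation and weapon foley sound class, with reduced volume and a limited audible distance.
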SoundCue /StaminaGameplay/Gameplay/Items/Hawaii/Audio/Hawaii_PickUp_Cue
root=WavePlayer_0
nodes [{id=WavePlayer_0, properties {soundWaveAssetPtr=&/Hawaii_PickUp}}]
properties {
 attenuationSettings=&Fort_Weapon_Pickup_Att
 maxDistance=3500
 soundClassObject=&Fort_WeaponFoley
 volumeMultiplier=0.251
}
