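; Gameplay ability that handles the equipped state of the Stamina Hawaii item, activated from a gameplay event.
; On the server only, it adds Stamina_Hawaii_Component to the avatar Fort player pawn if absent and applies the GE_Stamina_Hawaii_Equipped effect to the owner.
; It then waits for the Stamina.Hawaii.Equipped tag to be removed and ends itself, without removing the component.
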(event Ability|EventActivateAbilityFromEvent
  (params (EventData :type Gameplay%20Event%20Data%20Structure%20%28by%20ref%29))
  (Utilities|IsServer
    (:true
      (Utilities|IsValid (Actor|GetComponentbyClass (CallFunction|GetAvatarasFortPlayerPawn) "/StaminaGameplay/Gameplay/Items/Hawaii/Stamina_Hawaii_Component.Stamina_Hawaii_Component_C")
        (:Is%20Valid
          _)
        (:Is%20Not%20Valid
          (Game|AddComponentbyClass (CallFunction|GetAvatarasFortPlayerPawn) "/StaminaGameplay/Gameplay/Items/Hawaii/Stamina_Hawaii_Component.Stamina_Hawaii_Component_C")))
      (Ability|ApplyGameplayEffectToOwner "/StaminaGameplay/Gameplay/Items/Hawaii/GE_Stamina_Hawaii_Equipped.GE_Stamina_Hawaii_Equipped_C")
      (Ability|Tasks|WaitGameplayTagRemove "(TagName=\"Stamina.Hawaii.Equipped\")"
        (:Removed
          (Ability|EndAbility))))))
