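; GA_Stamina_Hawaii_TriggerFromAlaska is a gameplay ability for the Hawaii item, apparently triggered from Alaska, that knocks the avatar pawn out of skydiving.
; On activation it ends skydiving on the Fortnite player pawn and forces character movement into Falling.
; It then removes owner gameplay effects matching configured asset tags, and only the authority ends the ability.
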(event Ability|EventActivateAbility
  (bind _returnvalue (CallFunction|GetAvatarasFortPlayerPawn))
  (Parachute|EndSkydiving _returnvalue)
  (Pawn|Components|CharacterMovement|SetMovementMode (Variables|/Script/Engine.Character|GetCharacterMovement _returnvalue) "MOVE_Falling" 0)
  (Ability|RemoveGameplayEffectFromOwnerWithAssetTags (Variables|Default|GetWith%20Asset%20Tags))
  (Ability|HasAuthority
    (:True
      (Ability|EndAbility))))
