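; Gameplay ability for the Stamina Hawaii item that switches the weapon's camera targeting on while the ability input is held.
; On input press and release, it gets the avatar Fort player pawn's current weapon and casts it to B_Weap_Stamina_Empty.
; It then calls SetCameraTargeting on that weapon, passing true on press and false on release.
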(event Ability|EventOnAbilityInputPressed
  (bind _asb_weap_stamina_empty (Utilities|Casting|CastToB_Weap_Stamina_Empty (Weapon|GetCurrentWeapon (CallFunction|GetAvatarasFortPlayerPawn))))
  (Class|BWeapStaminaEmpty|SetCameraTargeting _asb_weap_stamina_empty true))

(event Ability|EventOnAbilityInputReleased
  (bind _asb_weap_stamina_empty (Utilities|Casting|CastToB_Weap_Stamina_Empty (Weapon|GetCurrentWeapon (CallFunction|GetAvatarasFortPlayerPawn))))
  (Class|BWeapStaminaEmpty|SetCameraTargeting _asb_weap_stamina_empty false))
