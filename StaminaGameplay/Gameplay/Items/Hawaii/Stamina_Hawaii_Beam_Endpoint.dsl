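; Endpoint actor for the Hawaii beam item that drives the beam's looping projectile and impact effects through local GameplayCues.
; BeginPlay starts the projectile cue; OnEndpointImpact swaps it for the impact cue, and OnEndpointHitNothing removes the impact cue without restoring it, gated by an endpoint state.
; A fixed timer auto-destroys the actor to prevent lingering impact FX.
; Cleanup, also run on destruction, removes both cues and clears that timer.
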(event EventBeginPlay
  (Projectile|AddGameplayCue%28Looping%29Local self (Variables|Default|GetProjectileCue) (Ability|GameplayCue|MakeGameplayCueParameters 0.0))
  ;[1 1 1] Auto-destroy if we're past the maximum time one of these should be alive to fix issue of persisting impact FX
  (bind _returnvalue (Utilities|Time|SetTimerbyEvent (EventDispatchers|CreateEvent :!event Custom|AutoDestroy) 8.5))
  (Variables|Default|SetAutoDestroyTimerHandle _returnvalue))

(event Collision|EventActorBeginOverlap
  (params (OtherActor :type Actor%20Object%20Reference)))

(event Custom|Cleanup
  ;[1 1 1] Cleanup
  (Projectile|RemoveGameplayCue%28Looping%29Local self (Variables|Default|GetLoopingImpactCue) (Ability|GameplayCue|MakeGameplayCueParameters 0.0))
  (Projectile|RemoveGameplayCue%28Looping%29Local self (Variables|Default|GetProjectileCue) (Ability|GameplayCue|MakeGameplayCueParameters 0.0))
  ;;[1 1 1] Clear timer in case we get cleaned up from the component instead of by auto-destroy
  (Utilities|Time|ClearandInvalidateTimerbyHandle (Variables|Default|GetAutoDestroyTimerHandle)))

(event Custom|AutoDestroy
  (CallFunction|Cleanup)
  (Actor|DestroyActor))

(event EventDestroyed
  (CallFunction|Cleanup))

(event EventEndPlay
  (params (EndPlayReason :type EEndPlayReason%20Enum :default "Destroyed")))

(fn OnEndpointHitNothing
  (if (== (Variables|Default|GetEndpointState) 1)
    (Projectile|RemoveGameplayCue%28Looping%29Local self (Variables|Default|GetLoopingImpactCue) (Ability|GameplayCue|MakeGameplayCueParameters 0.0) false)
    (Variables|Default|SetEndpointState 2)))

(fn OnEndpointImpact
  (if (!= (Variables|Default|GetEndpointState) 1)
    (Development|PrintString "Impact")
    (Projectile|RemoveGameplayCue%28Looping%29Local self (Variables|Default|GetProjectileCue) (Ability|GameplayCue|MakeGameplayCueParameters 0.0))
    (Projectile|AddGameplayCue%28Looping%29Local self (Variables|Default|GetLoopingImpactCue) (Ability|GameplayCue|MakeGameplayCueParameters 0.0))
    (Variables|Default|SetEndpointState 1)))

(fn ConstructionScript)
