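; Camera modifier for the Hawaii item that offsets the player's view position over a timed sequence, leaving rotation and FOV unchanged.
; Elapsed time drives a cursor and two curves, blending from the incoming view to a relative offset, then toward a second offset.
; When disabled, the result lerps with the original location by modifier alpha.
; The head socket is stored as a look target but unused in this graph.
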(fn BlueprintModifyCamera
  (params (DeltaTime :type Float%20%28single-precision%29) (ViewLocation :type Vector :default "0, 0, 0") (ViewRotation :type Rotator :default "0, 0, 0") (FOV :type Float%20%28single-precision%29))
  (Variables|Default|SetInLocation ViewLocation)
  (bind _asfort_player_pawn (Utilities|Casting|CastToFortPlayerPawn (Game|GetPlayerPawn 0)))
  (Variables|Default|SetPlayer _asfort_player_pawn)
  (Variables|Default|SetInRotation ViewRotation)
  (Variables|Default|SetElapsedTime (- (Utilities|Time|GetTimeSeconds) (Variables|Default|GetStartTime)))
  ;[1 1 1] Sequence 1
  (Variables|Default|SetSequenceCursor (Math|Float|Clamp(Float) (/ (Variables|Default|GetElapsedTime) (Variables|Default|GetSequence1Duration))))
  (Variables|Default|SetLookTarget (Transformation|GetSocketLocation (Variables|/Script/Engine.Character|GetMesh (Variables|Default|GetPlayer)) "FX_Head"))
  (bind _outnewlocation (Class|StaminaFunctionLibrary|ComputeRelativeViewLocation (Variables|Default|GetOffset_X) (Variables|Default|GetOffset_Y) (Variables|Default|GetOffset_Z) (Variables|Default|GetInLocation) (Variables|Default|GetInRotation)))
  (Variables|Default|SetTargetLocationSequence1 (Math|Vector|Lerp(Vector) (Variables|Default|GetInLocation) _outnewlocation (Math|Curves|GetFloatValue (Variables|Default|GetSequence1InitialBlendIn) (Variables|Default|GetSequenceCursor))))
  (bind _outnewlocation_2 (Class|StaminaFunctionLibrary|ComputeRelativeViewLocation (Variables|Default|GetOffset_X_Sequence1_Part2) (Variables|Default|GetOffset_Y_Sequence1_Part2) (Variables|Default|GetOffset_Z_Sequence1_Part2) (Variables|Default|GetInLocation) (Variables|Default|GetInRotation)))
  (Variables|Default|SetTargetLocationSequence1_Part2 _outnewlocation_2)
  (Variables|Default|SetFinalLocation (Math|Vector|Lerp(Vector) (Variables|Default|GetTargetLocationSequence1) (Variables|Default|GetTargetLocationSequence1_Part2) (Math|Curves|GetFloatValue (Variables|Default|GetSequence1LocationCurve) (Variables|Default|GetSequenceCursor))))
  (Variables|Default|SetFinalRot (Variables|Default|GetInRotation))
  (if (Variables|Default|GetDisabled)
    (Variables|Default|SetFinalLocation (Math|Vector|Lerp(Vector) (Variables|GetViewLocation) (Variables|Default|GetFinalLocation) (Variables|CameraModifier|GetAlpha))))
  (return (Variables|Default|GetFinalLocation) (Variables|Default|GetFinalRot) (Variables|GetFOV)))
