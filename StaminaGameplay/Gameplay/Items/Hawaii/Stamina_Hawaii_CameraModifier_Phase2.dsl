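; Camera modifier for phase 2 of the Hawaii item that shifts the view position and changes field of view as the modifier blends.
; BlueprintModifyCamera builds an offset location along the view rotation's right, forward and up vectors, leaving rotation unchanged.
; A float curve sampled at the modifier's Alpha eases the location blend.
; FOV lerps linearly by raw Alpha between local default and firing values, ignoring the incoming FOV.
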(fn BlueprintModifyCamera
  (params (DeltaTime :type Float%20%28single-precision%29) (ViewLocation :type Vector :default "0, 0, 0") (ViewRotation :type Rotator :default "0, 0, 0") (FOV :type Float%20%28single-precision%29))
  (Variables|Default|SetFinalLocation (+ (+ (+ ViewLocation (* (Math|Vector|GetRightVector (Variables|GetViewRotation)) (Variables|Default|GetOffset_X))) (* (Math|Vector|GetForwardVector (Variables|GetViewRotation)) (Variables|Default|GetOffset_Y))) (* (Math|Vector|GetUpVector (Variables|GetViewRotation)) (Variables|Default|GetOffset_Z))))
  (Variables|Default|SetFinalRot ViewRotation)
  (Variables|Default|SetFinalLocation (Math|Vector|Lerp(Vector) (Variables|GetViewLocation) (Variables|Default|GetFinalLocation) (Math|Curves|GetFloatValue (Variables|Default|GetInCurve) (Variables|CameraModifier|GetAlpha))))
  (LocalVariables|SetFinalFOV (Math|Float|Lerp (LocalVariables|GetDefaultFOV) (LocalVariables|GetFiringFOV) (Variables|CameraModifier|GetAlpha)))
  (return (Variables|Default|GetFinalLocation) (Variables|Default|GetFinalRot) (LocalVariables|GetFinalFOV)))
