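; Camera modifier for targeting in the Stamina Hawaii item set that pushes the player camera forward along its view direction.
; BlueprintModifyCamera offsets the view location by a fixed distance along the rotation's forward vector, then lerps from the original location by the modifier's Alpha for blending.
; Rotation passes through unchanged, and the FOV lerp adds a zero offset, so FOV is effectively untouched.
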(fn BlueprintModifyCamera
  (params (DeltaTime :type Float%20%28single-precision%29) (ViewLocation :type Vector :default "0, 0, 0") (ViewRotation :type Rotator :default "0, 0, 0") (FOV :type Float%20%28single-precision%29))
  (Variables|Default|SetFinalLoc (+ (Variables|GetViewLocation) (* (Math|Vector|GetForwardVector (Variables|GetViewRotation)) 6e+01)))
  (Variables|Default|SetFinalLoc (Math|Vector|Lerp(Vector) (Variables|GetViewLocation) (Variables|Default|GetFinalLoc) (Variables|CameraModifier|GetAlpha)))
  (return (Variables|Default|GetFinalLoc) (Variables|GetViewRotation) (Math|Float|Lerp FOV (+ FOV 0) (Variables|CameraModifier|GetAlpha))))
