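; Player pawn component for the Stamina Hawaii beam ability, handling aim, movement lock, cooldown state and cosmetics.
; While equipped, tick updates aimpoint and target-in-range and zeroes velocity during activation; the server adds feedback components to potential victims.
; Replicated beam data drives client-local gameplay cues, HUD hiding and a spawned endpoint actor, positioned by native trace or beam-battle intersection.
; Landing, death, water or riding restores the HUD and removes tagged effects.
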(event EventBeginPlay
  ;[1 1 1] BeginPlay
  (bind _asfort_player_pawn (Utilities|Casting|CastToFortPlayerPawn (Components|GetOwner)))
  (EventDispatchers|BindEventtoOnPawnLanded _asfort_player_pawn (delegate Custom|CustomEvent_0))
  (EventDispatchers|BindEventtoOnDied _asfort_player_pawn (delegate Custom|Died)))

(event Custom|CustomEvent_0
  (params (Hit :type Hit%20Result%20Structure%20%28by%20ref%29))
  ;[1 1 1] BeginPlay
  (CallFunction|PawnLanded))

(event Custom|Died
  (params (DamagedActor :type Actor%20Object%20Reference) (Damage :type Float%20%28single-precision%29) (InstigatedBy :type Controller%20Object%20Reference) (DamageCauser :type Actor%20Object%20Reference) (HitLocation :type Vector) (FHitComponent :type Primitive%20Component%20Object%20Reference) (BoneName :type Name) (Momentum :type Vector))
  ;[1 1 1] BeginPlay
  (CallFunction|PawnLanded))

(event EventTick
  (params (DeltaSeconds :type Float%20%28single-precision%29))
  (bind _player (Variables|Default|GetPlayer))
  (bind _returnvalue (Ability|GetAbilitySystemComponent _player))
  (bind _returnvalue_2 (GameplayTags|HasMatchingGameplayTag _returnvalue "(TagName=\"Granted.Cooldown.Stamina.Hawaii\")"))
  (bind _player_2 (Variables|Default|GetPlayer))
  ;[1 1 1] Tick
  (bind _asfort_player_pawn (Utilities|Casting|CastToFortPlayerPawn (Components|GetOwner)))
  (bind _output_get (Variables|Default|SetPlayer _asfort_player_pawn))
  (if (GameplayTags|HasMatchingGameplayTag _output_get "(TagName=\"Stamina.Hawaii.Equipped\")")
    (CallFunction|TickAimpoint)
    (CallFunction|LockinPlace)
    (Utilities|SwitchHasAuthority
      (:Authority
        (CallFunction|TickServer))
      (:Remote
        (CallFunction|TickClient)))
    (if (Variables|Default|GetBlockedByCooldown)
      _
      (elif (Variables|Default|GetAbilityActive)
        _
        (else
          (if (Variables|Default|GetAbilityActive)
            (Variables|Default|SetNewRotation (Math|Rotator|FindLookatRotation (Transformation|GetActorLocation (Variables|Default|GetPlayer)) (Variables|Default|GetAimpoint)))
            (bind _asfort_player_controller_athena (Utilities|Casting|CastToFortPlayerControllerAthena (Pawn|GetController _player_2)))
            (Class|FortPlayerControllerAthena|SetUseOverrideDesiredRotation _asfort_player_controller_athena true 1.0)
            (Class|FortPlayerControllerAthena|SetOverrideDesiredRotation _asfort_player_controller_athena (Variables|Default|GetNewRotation))
            (Variables|/Script/Engine.CharacterMovementComponent|SetRotationRate "300.000000,300.000000,300.000000" (Variables|/Script/Engine.Character|GetCharacterMovement _player_2))))))
    (if (!= _returnvalue_2 (Variables|Default|GetBlockedByCooldown))
      (Variables|Default|SetBlockedByCooldown _returnvalue_2)
      (else
        _))
    (else
      (CallFunction|TickEquippedGC))))

(fn AddEquipGC
  (GameplayCue|AddGameplayCueOnActor(Looping) (Variables|Default|GetPlayer) (Variables|Default|GetEquipGCTag) (Ability|GameplayCue|MakeGameplayCueParameters 0.0))
  (Variables|Default|SetHasEquipGC true))

(fn ApplyBeamVictimFeedback
  (params (Victim :type Fort%20Player%20Pawn%20Object%20Reference))
  (bind _localvictim (LocalVariables|GetLocalVictim))
  (bind _returnvalue (Actor|GetComponentbyClass _localvictim "/StaminaGameplay/Gameplay/Items/Hawaii/Stamina_Hawaii_Victim_Component.Stamina_Hawaii_Victim_Component_C"))
  (LocalVariables|SetLocalVictim Victim)
  (Variables|Default|SetDidClearVictimFeedback false)
  (Utilities|IsValid _returnvalue
    (:Is%20Valid
      _)
    (:Is%20Not%20Valid
      (bind _returnvalue_2 (Game|AddComponentbyClass (LocalVariables|GetLocalVictim) "/StaminaGameplay/Gameplay/Items/Hawaii/Stamina_Hawaii_Victim_Component.Stamina_Hawaii_Victim_Component_C"))))
  (Class|StaminaHawaiiVictimComponent|StartVictimFeedback _returnvalue (Transformation|GetActorLocation (Components|GetOwner)) (Variables|Default|GetPlayer)))

(fn CheckForCancel)

(fn CheckForSpawnImpactDecal
  (if (> (- (Utilities|Time|GetTimeSeconds) (Variables|Default|GetLastDecalTime)) 0.2)
    (GameplayCue|ExecuteGameplayCue%28Burst%29Local (Variables|Default|GetSpawnedEndpoint) (Variables|Default|GetDecalTag) (Ability|GameplayCue|MakeGameplayCueParameters 0.0))
    (Variables|Default|SetLastDecalTime (Utilities|Time|GetTimeSeconds))))

(fn ClearBeamVictimFeedback
  (Variables|Default|SetDidClearVictimFeedback true)
  (Utilities|Array|Clear (Variables|Default|GetVictimFeedbackPawns)))

(fn ClearCooldown
  (Variables|Default|SetBlockedByCooldown false))

(fn ComputeBeamEndpointForGC
  (bind _player (Variables|Default|GetPlayer))
  (bind _spawnedendpoint (Variables|Default|GetSpawnedEndpoint))
  (if (Variables|Default|GetRepInBeamBattle)
    (return (Variables|Default|GetRepBattleIntersectionPoint) (Math|Vector|Normalize (- (Transformation|GetActorLocation (Variables|Default|GetPlayer)) (Variables|Default|GetRepBattleIntersectionPoint))) false)
    (else
      (Utilities|IsValid (Variables|Default|GetPlayer)
        (:Is%20Valid
          (if (Pawn|IsLocallyControlled _player)
            (LocalVariables|SetInAimpoint (Variables|Default|GetAimpoint))
            (else
              (LocalVariables|SetInAimpoint (+ (Transformation|GetActorLocation _player) (* (Math|Vector|GetForwardVector (Pawn|GetBaseAimRotation _player)) 2e+04)))))
          (bind (_outendpoint _outimpact _outatmaxrange _outimpactnormal) (Class|StaminaFunctionLibraryNative|TraceforBeamEndpointNative (Variables|Default|GetPlayer) (ScalableFloat|GetValueatLevel (Variables|Default|GetRow_BeamSpeed)) (ScalableFloat|GetValueatLevel (Variables|Default|GetRow_BeamRange)) "FX_Chest" (Variables|Default|GetBeamStartTime) "FortTrajectoryBlocksWater" (LocalVariables|GetInAimpoint)))
          (if _outimpact
            (Class|StaminaHawaiiBeamEndpoint|OnEndpointImpact _spawnedendpoint)
            (elif _outatmaxrange
              (Class|StaminaHawaiiBeamEndpoint|OnEndpointHitNothing _spawnedendpoint)
              (else
                _)))
          (bind _output_get (Variables|Default|SetBeamImpacted _outimpact))
          (Variables|Default|SetBeamAtMaxRange _outatmaxrange)
          (return _outendpoint _outimpactnormal _output_get))))))

(fn DestroyEndpoint
  (bind _spawnedendpoint (Variables|Default|GetSpawnedEndpoint))
  (Utilities|IsValid _spawnedendpoint
    (:Is%20Valid
      (Actor|DestroyActor _spawnedendpoint))
    (:Is%20Not%20Valid
      (Development|PrintString "no endpoint?"))))

(fn HandleNewBeamData
  (bind _repbeamdata (Variables|Default|GetRepBeamData))
  (bind (_beamactive_1_74f47b0849047809ab388489df33bd1b _abilityactive_7_97a41cf944e93caf32bea3a685a6da71 _beamendpoint_10_aed050e043239d1a896696b4261f3973) (Utilities|Struct|BreakStaminaHawaiiBeamData _repbeamdata))
  (bind _asfort_player_pawn (Utilities|Casting|CastToFortPlayerPawn (Components|GetOwner)))
  (Variables|Default|SetPlayer _asfort_player_pawn)
  (if (and (not (Variables|Default|GetAbilityActive)) _abilityactive_7_97a41cf944e93caf32bea3a685a6da71)
    (Variables|Default|SetAbilityStartTime (Utilities|Time|GetTimeSeconds))
    (Variables|Default|SetSpawnedBeam false)
    (Variables|Default|SetDidSpawnGlint false)
    (Variables|Default|SetDidCancel false)
    (Variables|Default|SetAbilityStartPosition (Transformation|GetActorLocation (Variables|Default|GetPlayer)))
    (else
      _))
  (Variables|Default|SetLocalBeamActive _beamactive_1_74f47b0849047809ab388489df33bd1b)
  (Variables|Default|SetAbilityActive _abilityactive_7_97a41cf944e93caf32bea3a685a6da71)
  (Variables|Default|SetBeamEndpoint _beamendpoint_10_aed050e043239d1a896696b4261f3973)
  (if (Utilities|Enum|Equal(Enum) (Networking|GetLocalRole (Variables|Default|GetPlayer)) "ROLE_SimulatedProxy")
    (if (not (Variables|Default|GetDidSpawnGlint))
      (Projectile|AddGameplayCue%28Looping%29Local (Components|GetOwner) (Variables|Default|GetGC_Glint) (Ability|GameplayCue|MakeGameplayCueParameters 0.0))
      (Variables|Default|SetDidSpawnGlint true))
    (else
      _))
  (Utilities|FlowControl|Sequence
    (:then_0
      (if (Networking|IsServer)
        _
        (elif (Math|Boolean|ANDBoolean (Variables|Default|GetAbilityActive) (Variables|Default|GetLocalBeamActive) (not (Variables|Default|GetSpawnedBeam)))
          (Variables|Default|SetSpawnedBeam true)
          (Variables|Default|SetBeamStartTime (Utilities|Time|GetTimeSeconds))
          (Projectile|AddGameplayCue%28Looping%29Local (Components|GetOwner) (Variables|Default|GetGC_Beam) (Ability|GameplayCue|MakeGameplayCueParameters 0.0))
          (bind _outbeamend (CallFunction|ComputeBeamEndpointforGC))
          (bind _returnvalue (Game|SpawnActorfromClass "/StaminaGameplay/Gameplay/Items/Hawaii/Stamina_Hawaii_Beam_Endpoint.Stamina_Hawaii_Beam_Endpoint_C" (Math|Conversions|ToTransform(Vector) _outbeamend) "AlwaysSpawn" "OverrideRootScale" 0 0))
          (Variables|Default|SetSpawnedEndpoint _returnvalue)
          (bind _asfort_player_controller_gameplay (Utilities|Casting|CastToFortPlayerControllerGameplay (Fortnite|GetFortPlayerControllerfromActor (Components|GetOwner))))
          (HUD|SetHUDElementVisibilitywithContext _asfort_player_controller_gameplay (GameplayTags|MakeGameplayTagContainerfromTag "(TagName=\"HUD.Athena.Hide.WithInventory\")") true "Ability" "")
          (Variables|Default|SetDidLand false)
          (GameplayCue|ExecuteGameplayCue%28Burst%29Local (Components|GetOwner) (Variables|Default|GetFireRumbleTag) (Ability|GameplayCue|MakeGameplayCueParameters 0.0)))))
    (:then_1
      (if (not (Variables|Default|GetLocalBeamActive))
        (Projectile|RemoveGameplayCue%28Looping%29Local (Components|GetOwner) (Variables|Default|GetGC_Beam) (Ability|GameplayCue|MakeGameplayCueParameters 0.0))
        (Projectile|RemoveGameplayCue%28Looping%29Local (Components|GetOwner) (Variables|Default|GetGC_Glint) (Ability|GameplayCue|MakeGameplayCueParameters 0.0))
        (Utilities|IsValid (Variables|Default|GetSpawnedEndpoint)
          (:Is%20Valid
            (Class|StaminaHawaiiBeamEndpoint|Cleanup (Variables|Default|GetSpawnedEndpoint))
            (Utilities|Time|SetTimerbyEvent (EventDispatchers|CreateEvent :!event DestroyEndpoint) 1.0)
            (Utilities|Time|SetTimerbyEvent (EventDispatchers|CreateEvent :!event PawnLanded) 0.2))
          (:Is%20Not%20Valid
            (Development|PrintString "Warning - no endpoint spawned")))))
    (:then_2
      (if (not (Variables|Default|GetAbilityActive))
        (bind _asfort_player_controller_athena (Utilities|Casting|CastToFortPlayerControllerAthena (Pawn|GetController (Variables|Default|GetPlayer))))
        (Class|FortPlayerControllerAthena|SetUseOverrideDesiredRotation _asfort_player_controller_athena false 1.0)
        (Utilities|Time|SetTimerbyEvent (EventDispatchers|CreateEvent :!event ResetPawnRotationRate) 1.0)))))

(event Custom|AbilityFinishedUnhide)

(fn LockInPlace
  (if (and (Variables|Default|GetAbilityActive) (> (- (Utilities|Time|GetTimeSeconds) (Variables|Default|GetAbilityStartTime)) 0.25))
    (Variables|/Script/Engine.MovementComponent|SetVelocity "0, 0, 0" (Variables|/Script/Engine.Character|GetCharacterMovement (Variables|Default|GetPlayer)))))

(fn OnRep_BlockedByCooldown
  (Default|CallOnCooldownState (Variables|Default|GetBlockedByCooldown)))

(fn OnRep_RepAbilityActive
  (if (Utilities|Enum|Equal(Enum) (Networking|GetLocalRole (Components|GetOwner)) "ROLE_SimulatedProxy")
    (if (not (Variables|Default|GetAbilityActive))
      (Transformation|SetActorRotation (Components|GetOwner) "0, 0, 0" false)
      (bind _asfort_player_controller_athena (Utilities|Casting|CastToFortPlayerControllerAthena (Pawn|GetController (Variables|Default|GetPlayer))))
      (Class|FortPlayerControllerAthena|SetUseOverrideDesiredRotation _asfort_player_controller_athena false 1.0))))

(fn OnRep_RepBeamData
  (CallFunction|HandleNewBeamData))

(fn PawnLanded
  (bind _returnvalue (Components|GetOwner))
  (Variables|Default|SetDidLand true)
  (bind _asfort_player_controller_gameplay (Utilities|Casting|CastToFortPlayerControllerGameplay (Fortnite|GetFortPlayerControllerfromActor _returnvalue)))
  (HUD|SetHUDElementVisibilitywithContext _asfort_player_controller_gameplay (GameplayTags|MakeGameplayTagContainerfromTag "(TagName=\"HUD.Athena.Hide.WithInventory\")") false "Ability" "")
  (GameplayEffects|RemoveActiveEffectswithGrantedTags (Ability|GetAbilitySystemComponent _returnvalue) (Variables|Default|GetRemovalLandedTag)))

(fn RemoveEquipGC
  (GameplayCue|RemoveGameplayCueOnActor(Looping) (Variables|Default|GetPlayer) (Variables|Default|GetEquipGCTag) (Ability|GameplayCue|MakeGameplayCueParameters 0.0))
  (Variables|Default|SetHasEquipGC false))

(fn ResetPawnRotationRate
  (bind _player (Variables|Default|GetPlayer))
  (Utilities|IsValid _player
    (:Is%20Valid
      (Variables|/Script/Engine.CharacterMovementComponent|SetRotationRate "-1.000000,-1.000000,-1.000000" (Variables|/Script/Engine.Character|GetCharacterMovement _player)))))

(fn SetPlayerView
  (bind _player (Variables|Default|GetPlayer))
  (bind _returnvalue (Pawn|GetController _player))
  (bind (_location _rotation) (Pawn|GetPlayerViewPoint _returnvalue))
  (bind _player_2 (Variables|Default|GetPlayer))
  (bind _returnvalue_2 (Pawn|GetController _player_2))
  (Utilities|IsValid _player
    (:Is%20Valid
      (Utilities|IsValid _returnvalue
        (:Is%20Valid
          (Variables|Default|SetCameraLocation _location)
          (Variables|Default|SetCameraForward (Math|Vector|GetForwardVector _rotation))
          (Variables|Default|SetUseAimpoint (Variables|Default|GetBeamEndpoint))
          (if (< (Math|Vector|DistanceSquared(Vector) (Variables|Default|GetCameraLocation) (Variables|Default|GetBeamEndpoint)) (* (Variables|Default|GetAimTooCloseThreshold) (Variables|Default|GetAimTooCloseThreshold)))
            (Variables|Default|SetUseAimpoint (+ (Variables|Default|GetCameraLocation) (* (Variables|Default|GetCameraForward) 1e+03)))
            (else
              _))
          (Fortnite|KeepCGPathOptimizations _returnvalue_2)
          (Pawn|SetControlRotation _returnvalue_2 (Math|Rotator|FindLookatRotation (Variables|Default|GetCameraLocation) (Variables|Default|GetUseAimpoint))))))))

(fn TickEquippedGC
  (Utilities|FlowControl|Sequence
    (:then_0
      (if (GameplayTags|HasMatchingGameplayTag (Variables|Default|GetPlayer) "(TagName=\"Stamina.Hawaii.Equipped\")")
        (if (and (not (Weapon|IsWeaponHolstered (Variables|Default|GetPlayer))) (not (Variables|Default|GetHasEquipGC)))
          (CallFunction|AddEquipGC))
        (elif (Variables|Default|GetHasEquipGC)
          (CallFunction|RemoveEquipGC))))
    (:then_1
      (if (GameplayTags|HasMatchingGameplayTag (Variables|Default|GetPlayer) "(TagName=\"Stamina.Hawaii.Equipped\")")
        (if (and (Weapon|IsWeaponHolstered (Variables|Default|GetPlayer)) (Variables|Default|GetHasEquipGC))
          (CallFunction|RemoveEquipGC))))))

(fn TickAimpoint
  (bind _player (Variables|Default|GetPlayer))
  (bind _returnvalue (Pawn|GetController _player))
  (bind _player_2 (Variables|Default|GetPlayer))
  (bind _returnvalue_2 (Pawn|GetController _player_2))
  (Utilities|IsValid (Variables|Default|GetPlayer)
    (:Is%20Valid
      (Utilities|IsValid _returnvalue_2
        (:Is%20Valid
          (bind _returnvalue_3 (Fortnite|GetPlayerAimPointHit _returnvalue_2 (ScalableFloat|GetValueatLevel (Variables|Default|GetRow_BeamRange)) (Utilities|Array|MakeArray 0)))
          ;[1 1 1] Update for UI
          (Variables|Default|SetHasTargetInRange (Collision|BreakHitResult _returnvalue_3))
          (if (Variables|Default|GetAbilityActive)
            (if (or (not (Collision|BreakHitResult _returnvalue_3)) (< (Math|Vector|Distance(Vector) (.ImpactPoint (Collision|BreakHitResult _returnvalue_3)) (Transformation|GetActorLocation (Variables|Default|GetPlayer))) 5e+02))
              (Utilities|IsValid _returnvalue
                (:Is%20Valid
                  (bind _returnvalue_4 (Fortnite|GetPlayerAimPoint _returnvalue (ScalableFloat|GetValueatLevel (Variables|Default|GetRow_BeamRange)) (Utilities|Array|MakeArray 0) true))
                  (Variables|Default|SetAimpoint _returnvalue_4)))
              (else
                (Variables|Default|SetAimpoint (.ImpactPoint (Collision|BreakHitResult _returnvalue_3)))))))))))

(fn TickBeamVictimFeedback
  (if (Variables|Default|GetAbilityActive)
    (if (> (- (Utilities|Time|GetTimeSeconds) (Variables|Default|GetAbilityStartTime)) (ScalableFloat|GetValueatLevel (Variables|Default|GetWarningDelay)))
      (bind _asfort_player_pawn (Utilities|Casting|CastToFortPlayerPawn (Components|GetOwner)))
      (bind _outtargets (Class|StaminaFunctionLibrary|GetPlayerPossibleTargets _asfort_player_pawn 3e+02 (ScalableFloat|GetValueatLevel (Variables|Default|GetRow_BeamRange))))
      (for _array_element _outtargets
        (LocalVariables|SetPossibleNewVictim _array_element)
        (if (not (Utilities|Array|ContainsItem (Variables|Default|GetVictimFeedbackPawns) (LocalVariables|GetPossibleNewVictim)))
          (Utilities|Array|Add (LocalVariables|GetNewVictimFeedbackPawns) (LocalVariables|GetPossibleNewVictim))))
      (for _array_element_2 (LocalVariables|GetNewVictimFeedbackPawns)
        (CallFunction|ApplyBeamVictimFeedback _array_element_2))
      (Utilities|Array|AppendArray (Variables|Default|GetVictimFeedbackPawns) (LocalVariables|GetNewVictimFeedbackPawns)))
    (elif (Variables|Default|GetDidClearVictimFeedback)
      _
      (else
        (CallFunction|ClearBeamVictimFeedback)))))

(fn TickClient
  (bind _player (Variables|Default|GetPlayer))
  (if (Pawn|IsLocallyViewed (Variables|Default|GetPlayer))
    (Default|CallOnHasTargetinRange (Variables|Default|GetHasTargetInRange))
    (else
      _))
  (if (and (Variables|Default|GetAbilityActive) (Variables|Default|GetLocalBeamActive))
    (Utilities|IsValid (Variables|Default|GetSpawnedEndpoint)
      (:Is%20Valid
        (bind (_outbeamend _outimpactnormal _beamimpacted) (CallFunction|ComputeBeamEndpointforGC))
        (Variables|Default|SetBeamEndForGC _outbeamend)
        (if _beamimpacted
          (LocalVariables|SetEndpointRot (Math|Conversions|RotationFromXVector _outimpactnormal))
          (Transformation|SetActorLocationAndRotation (Variables|Default|GetSpawnedEndpoint) (Variables|Default|GetBeamEndForGC) (LocalVariables|GetEndpointRot) false false)
          (CallFunction|CheckforSpawnImpactDecal)
          (else
            (LocalVariables|SetEndpointRot (Math|Rotator|FindLookatRotation (Variables|Default|GetBeamEndForGC) (Transformation|GetSocketLocation (Variables|/Script/Engine.Character|GetMesh (Variables|Default|GetPlayer)) "FX_Chest")))
            (Transformation|SetActorLocationAndRotation (Variables|Default|GetSpawnedEndpoint) (Variables|Default|GetBeamEndForGC) (LocalVariables|GetEndpointRot) false false)))))
    (else
      (Utilities|FlowControl|Sequence
        (:then_0
          (Utilities|IsValid (WaterBody|GetCurrentWaterBody _player)
            (:Is%20Valid
              (if (not (Variables|Default|GetDidLand))
                (CallFunction|PawnLanded)))))
        (:then_1
          (if (FortMovement|IsRidingonActor _player)
            (if (not (Variables|Default|GetDidLand))
              (CallFunction|PawnLanded))))))))

(fn TickServer
  (CallFunction|TickBeamVictimFeedback)
  (CallFunction|TickEquippedGC)
  (CallFunction|CheckforCancel))

(fn TickSprintInterrupt)
