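; Victim-side component for the Stamina Hawaii item that plays hit feedback on its owning pawn.
; StartVictimFeedback stores the shooter, then, unless feedback is already active, resets the lens count and holds a local looping gameplay cue for three seconds.
; SpawnMultipleLensEffects, not called within this component, uses a self-rescheduling timer to spawn up to three shield-damage directional camera lens effects sourced from the shooter.
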(event Custom|StartVictimFeedback
  (params (ShooterLocation :type Vector) (Shooter :type Fort%20Player%20Pawn%20Object%20Reference))
  (Variables|Default|SetShooter Shooter)
  (if (not (Variables|Default|GetFeedbackActive))
    (Variables|Default|SetFeedbackActive true)
    (Variables|Default|SetLensEffectCount 0)
    (Projectile|AddGameplayCue%28Looping%29Local (Components|GetOwner) (Variables|Default|GetVictimTag) (Ability|GameplayCue|MakeGameplayCueParameters :NormalizedMagnitude 0.0 :RawMagnitude 0.0 :Location "0,0,0" :Normal "0,0,0" :EffectCauser Shooter :GameplayEffectLevel 1 :AbilityLevel 1 :bReplicateLocationWhenUsingMinimalRepProxy false))
    (Utilities|FlowControl|Delay 3.0)
    (Projectile|RemoveGameplayCue%28Looping%29Local (Components|GetOwner) (Variables|Default|GetVictimTag) (Ability|GameplayCue|MakeGameplayCueParameters 0.0))
    (Variables|Default|SetFeedbackActive false)))

(fn SpawnLensEffect
  (bind _asfort_pawn (Utilities|Casting|CastToFortPawn (Components|GetOwner)))
  (Feedback|SpawnGenericCameraLensEffectDirectional _asfort_pawn "/Game/Effects/Fort_Effects/Camera_Effects/B_PlayerShieldDamage_LensEffect_Direction.B_PlayerShieldDamage_LensEffect_Direction_C" (Variables|Default|GetShooter) (Collision|MakeHitResult false) false))

(fn SpawnMultipleLensEffects
  (if (< (Variables|Default|GetLensEffectCount) 3)
    (CallFunction|SpawnLensEffect)
    (Math|Integer|IncrementInt (Variables|Default|GetLensEffectCount)
      (:%20%20
        (Utilities|Time|SetTimerbyEvent (EventDispatchers|CreateEvent :!event SpawnMultipleLensEffects) 0.75)))))
